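; Passive quest ability that credits surviving storm circles while the player carries a qualifying Mythic weapon (Galileo quest).
; On each storm zone update, it gets Mythic items from the owning player's inventory and checks them against the required weapon tags.
; A match swaps the quest source tag to the Galileo variant before the parent handler runs.
; Afterward the tag resets to the generic SurviveStormCircles tag, so the swap applies only to that update.
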(event Ability|EventFailedToActivatePassiveAbility
  (params (ActorInfo :type Gameplay%20Ability%20Actor%20Info%20Structure))
  (|CallParentFunction))

(event Custom|CustomEvent
  (|CallParentFunction))

(event EventStormZoneUpdated
  (params (SafeZonePhase :type Integer))
  ;[1 1 1] When storm zone updates, check if player has a weapon of any of the specified rarity types in inventory
  (Variables|Default|SetOwningPlayerInventoryInterface (Fortnite|GetFortPlayerControllerfromActor (Variables|GAVars|Data|GetOwningPlayerPawn)))
  (bind _items (InventoryManagement|GetItemsfromInventoryOwnerforRarity (Variables|Default|GetOwningPlayerInventoryInterface) "Mythic"))
  (bind _array_element (Utilities|Array|ForEachLoopwithBreak _items)
    (:LoopBody
      (if (GameplayTags|HasAnyTags (GameplayTags|GetOwnedGameplayTags _array_element :!class /Script/FortniteGame.FortWorldItem) (Variables|Default|GetT%20Required%20Weapon%20Tags))
        (Variables|Default|SetQuestSourceTag "(GameplayTags=((TagName=\"Athena.Quests.SurviveStormCircles.Galileo\")))")))
    (:Completed
      _))
  (|CallParentFunction 0)
  (Variables|Default|SetQuestSourceTag "(GameplayTags=((TagName=\"Athena.Quests.SurviveStormCircles\")))"))
